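; PuffyStraw shield actor for an NPC boss, handling the shield's visual and sound effects and its link to its owner.
; Game-time material parameters drive fade-in, hit flashes and fade-out alongside Niagara effects.
; Death sound differs for destruction versus self-retraction, based on instigator validity.
; AttachShield sets ShieldOwner, whose OnRep attaches the shield and binds the owner's death to apply a damage effect to itself and destroy it.
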(event EventBeginPlay
  (bind _returnvalue (Utilities|Time|GetGameTimeinSeconds))
  (bind _sideways_shieldmesh (Variables|Default|GetSideways_ShieldMesh))
  (Networking|CanExecuteCosmeticEvents
    (:True
      (Rendering|Material|SetScalarParameterValueonMaterials _sideways_shieldmesh "FadeInTime" _returnvalue)
      (Rendering|Material|SetScalarParameterValueonMaterials _sideways_shieldmesh "FadeOutTime" (+ _returnvalue 100000000376832.0))
      (Utilities|FlowControl|Delay 0.6)
      (Components|Activation|Activate (Variables|Default|GetNS_PuffyStraw_Idle) false))))

(event EventOnDeathPlayEffects
  (params (Damage :type Float%20%28single-precision%29) (DamageTags :type Gameplay%20Tag%20Container%20Structure%20%28by%20ref%29) (Momentum :type Vector :default "0, 0, 0") (HitInfo :type Hit%20Result%20Structure%20%28by%20ref%29) (InstigatedBy :type Fort%20Pawn%20Object%20Reference) (DamageCauser :type Actor%20Object%20Reference) (EffectContext :type Gameplay%20Effect%20Context%20Handle%20Structure))
  (Utilities|IsValid InstigatedBy
    (:Is%20Valid
      (Development|PrintString "Something Destroyed the SHield")
      (Variables|Sound|SetShield%20Death%20Sound "/NPCLibrary/Audio/PuffyStraw/PuffyStraw_Destroyed_Cue.PuffyStraw_Destroyed_Cue"))
    (:Is%20Not%20Valid
      (Development|PrintString "Boss Removed Her Shield By herself")
      (Variables|Sound|SetShield%20Death%20Sound "/NPCLibrary/Audio/PuffyStraw/PuffyStraw_Retract_Cue.PuffyStraw_Retract_Cue")))
  (Audio|PlaySoundatLocation (Variables|Sound|GetShield%20Death%20Sound) (Transformation|GetWorldLocation (Variables|Default|GetSideways_ShieldMesh)))
  (Rendering|Material|SetScalarParameterValueonMaterials (Variables|Default|GetSideways_ShieldMesh) "FadeOutTime" (Utilities|Time|GetGameTimeinSeconds))
  (bind (_direction _fade_alpha) (Timeline|Fades%20Alpha%20Idle%20Particles%20%28matches%20Fade%20Time%20in%20Material%29 PlayFromStart)
    (:Update
      (Effects|Components|ParticleSystem|SetFloatParameter (Variables|Default|GetNS_PuffyStraw_Idle) "Alpha" _fade_alpha))
    (:Finished
      (Utilities|FlowControl|Delay 0.666)
      (Components|DestroyComponent (Variables|Default|GetSideways_ShieldMesh)))))

(event EventOnDamagePlayEffects
  (params (Damage :type Float%20%28single-precision%29) (DamageTags :type Gameplay%20Tag%20Container%20Structure%20%28by%20ref%29) (Momentum :type Vector :default "0, 0, 0") (HitInfo :type Hit%20Result%20Structure%20%28by%20ref%29) (InstigatedBy :type Fort%20Pawn%20Object%20Reference) (DamageCauser :type Actor%20Object%20Reference) (EffectContext :type Gameplay%20Effect%20Context%20Handle%20Structure))
  (bind (_bblockinghit _binitialoverlap _time _distance _location _impactpoint _normal _impactnormal _physmat _hitactor _hitcomponent _hitbonename _bonename _hititem _elementindex _faceindex _tracestart _traceend) (Collision|BreakHitResult HitInfo))
  (bind _returnvalue (Utilities|Time|GetGameTimeinSeconds))
  (bind _sideways_shieldmesh (Variables|Default|GetSideways_ShieldMesh))
  (Niagara|SpawnSystematLocation "/NPCLibrary/Abilities/PuffyStraw/FX/NS_PuffyStraw_Impacts.NS_PuffyStraw_Impacts" _location (Math|Rotator|MakeRotfromX (Math|Vector|GetReflectionVector (- _traceend _tracestart) _impactnormal)))
  (Rendering|Material|SetScalarParameterValueonMaterials _sideways_shieldmesh "HitFadeInTime" _returnvalue)
  (Rendering|Material|SetScalarParameterValueonMaterials _sideways_shieldmesh "HitFadeOutTime" (+ _returnvalue 0.125))
  (Audio|PlaySoundatLocation "/NPCLibrary/Audio/PuffyStraw/PuffyStraw_Impact_Damage_Cue.PuffyStraw_Impact_Damage_Cue" (Transformation|GetWorldLocation _sideways_shieldmesh)))

(event Custom|AttachShield
  (params (ShieldOwner :type Actor%20Object%20Reference))
  (Variables|Default|SetShieldOwner ShieldOwner))

(event Custom|ValidateHolder
  (bind _shieldowner (Variables|Default|GetShieldOwner)
    (:then
      (bind _asfort_pawn (Utilities|Casting|CastToFortPawn _shieldowner))
      (EventDispatchers|BindEventtoOnDied _asfort_pawn (delegate Custom|PawnDied)))
    (:else)))

(event Custom|PawnDied
  (params (DamagedActor :type Actor%20Object%20Reference) (Damage :type Float%20%28single-precision%29) (InstigatedBy :type Controller%20Object%20Reference) (DamageCauser :type Actor%20Object%20Reference) (HitLocation :type Vector) (FHitComponent :type Primitive%20Component%20Object%20Reference) (BoneName :type Name) (Momentum :type Vector))
  ;[1 1 1] When Owner Dies, make sure to remove shield if it's alive
  (GameplayEffects|ApplyGameplayEffectToTarget :self (Ability|GetAbilitySystemComponent self) :GameplayEffectClass "/Game/Athena/Items/Consumables/ShockwaveGrenade/GE_ShockGrenade_Damage.GE_ShockGrenade_Damage_C" :Target (Ability|GetAbilitySystemComponent self) :Level 0.0)
  (Actor|DestroyActor))

(fn OnRep_ShieldOwner
  (bind _shieldowner (Variables|Default|GetShieldOwner))
  (Utilities|IsValid _shieldowner
    (:Is%20Valid
      (Transformation|AttachActorToActor self _shieldowner "None" "KeepWorld" "KeepWorld" "KeepWorld" true)
      (CallFunction|ValidateHolder))))

(fn ConstructionScript)
