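; Shield actor in the NPC library that forms on spawn, attaches to an owning pawn and ignores movement collision with its capsule until EndPlay.
; Setting ShieldOwner drives attachment through its OnRep and binds the owner's death, which applies shockwave grenade damage to the shield to remove it.
; Hits send damage-scaled ripple parameters to mesh materials, spawn impact FX and open the core.
; Death effects erode the shield and shut off summon FX and post-process.
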(event EventBeginPlay
  (Networking|CanExecuteCosmeticEvents
    (:True
      (bind (_direction _formingcurve _formingcurveb) (Timeline|ShieldForming PlayFromStart)
        (:Update
          (Rendering|Material|SetScalarParameterValueonMaterials (Variables|Default|GetShieldMesh) "Brightness" (Math|Float|Lerp 1.0 4e+01 _formingcurve))
          (Rendering|Material|SetScalarParameterValueonMaterials (Variables|Default|GetCoreMesh) "EmissiveBrightness" (Math|Float|Lerp 1.0 8e+01 _formingcurveb)))))))

(event EventOnDamagePlayEffects
  (params (Damage :type Float%20%28single-precision%29) (DamageTags :type Gameplay%20Tag%20Container%20Structure%20%28by%20ref%29) (Momentum :type Vector :default "0, 0, 0") (HitInfo :type Hit%20Result%20Structure%20%28by%20ref%29) (InstigatedBy :type Fort%20Pawn%20Object%20Reference) (DamageCauser :type Actor%20Object%20Reference) (EffectContext :type Gameplay%20Effect%20Context%20Handle%20Structure))
  (bind (_bblockinghit _binitialoverlap _time _distance _location _impactpoint _normal _impactnormal) (Collision|BreakHitResult HitInfo))
  (bind _hitwavenormal (Variables|Default|GetHitWaveNormal))
  (bind _hitwavelocation (Variables|Default|GetHitWaveLocation))
  ;[1 1 1] Hit reacitve VFX
  (Utilities|FlowControl|Sequence
    (:then_0
      (Variables|Default|SetDamage Damage)
      (Variables|Default|SetHitWaveLocation _impactpoint)
      (Variables|Default|SetHitWaveNormal _impactnormal)
      (CallFunction|SetRippleFXParam _hitwavelocation (* (Math|Float|Lerp 3e+01 2.4e+02 (Math|Float|NormalizetoRange (Variables|Default|GetDamage) 5.0 45.0)) 1.2) (* (Math|Float|Lerp 1e+02 8e+02 (Math|Float|NormalizetoRange (Variables|Default|GetDamage) 5.0 45.0)) 1.5) (Math|Float|Lerp 0.25 0.45 (Math|Float|NormalizetoRange (Variables|Default|GetDamage) 5.0 45.0)) _hitwavenormal)
      (Effects|Components|ParticleSystem|SpawnFXAtLocation (Variables|Default|GetImpactFX) _hitwavelocation (Math|Conversions|RotationFromXVector _hitwavenormal)))
    (:then_1
      (if (not (Variables|Default|GetIsOpened))
        (bind (_direction _open) (Timeline|ShieldOpen PlayFromStart)
          (:Update
            (Rendering|Material|SetScalarParameterValueonMaterials (Variables|Default|GetCoreMesh) "ShieldOpen" _open))
          (:IsOpened
            (Variables|Default|SetIsOpened true))
          (:IsClosed
            (Variables|Default|SetIsOpened false)))))))

(event EventOnDeathPlayEffects
  (params (Damage :type Float%20%28single-precision%29) (DamageTags :type Gameplay%20Tag%20Container%20Structure%20%28by%20ref%29) (Momentum :type Vector :default "0, 0, 0") (HitInfo :type Hit%20Result%20Structure%20%28by%20ref%29) (InstigatedBy :type Fort%20Pawn%20Object%20Reference) (DamageCauser :type Actor%20Object%20Reference) (EffectContext :type Gameplay%20Effect%20Context%20Handle%20Structure))
  (Components|Activation|Deactivate (Variables|Default|GetNS_CreatureRise_Summon))
  (if (Variables|Default|GetUseDecal)
    (Rendering|SetVisibility (Variables|Default|GetPointLight) false false)
    (else
      _))
  (Variables|/Script/Engine.PostProcessComponent|SetbEnabled false (Variables|Default|GetPostProcess))
  (bind (_direction _erode) (Timeline|ErodeOut PlayFromStart)
    (:Update
      (Rendering|Material|SetScalarParameterValueonMaterials (Variables|Default|GetShieldMesh) "Erode" _erode))))

(event Custom|AttachShield
  (params (ShieldOwner :type Actor%20Object%20Reference))
  (Variables|Default|SetShieldOwner ShieldOwner))

(event Custom|ValidateHolder
  (bind _shieldowner (Variables|Default|GetShieldOwner)
    (:then
      (bind _asfort_pawn (Utilities|Casting|CastToFortPawn _shieldowner))
      (EventDispatchers|BindEventtoOnDied _asfort_pawn (delegate Custom|PawnDied))
      (Collision|IgnoreActorwhenMoving (Variables|/Script/Engine.Character|GetCapsuleComponent _asfort_pawn) self true))
    (:else)))

(event Custom|PawnDied
  (params (DamagedActor :type Actor%20Object%20Reference) (Damage :type Float%20%28single-precision%29) (InstigatedBy :type Controller%20Object%20Reference) (DamageCauser :type Actor%20Object%20Reference) (HitLocation :type Vector) (FHitComponent :type Primitive%20Component%20Object%20Reference) (BoneName :type Name) (Momentum :type Vector))
  ;[1 1 1] When Owner Dies, make sure to remove shield if it's alive
  (GameplayEffects|ApplyGameplayEffectToTarget :self (Ability|GetAbilitySystemComponent self) :GameplayEffectClass "/Game/Athena/Items/Consumables/ShockwaveGrenade/GE_ShockGrenade_Damage.GE_ShockGrenade_Damage_C" :Target (Ability|GetAbilitySystemComponent self) :Level 0.0))

(event EventEndPlay
  (params (EndPlayReason :type EEndPlayReason%20Enum :default "Destroyed"))
  (bind _shieldowner (Variables|Default|GetShieldOwner))
  (Utilities|IsValid _shieldowner
    (:Is%20Valid
      (bind _asfort_pawn (Utilities|Casting|CastToFortPawn _shieldowner))
      (Collision|IgnoreActorwhenMoving (Variables|/Script/Engine.Character|GetCapsuleComponent _asfort_pawn) self false))))

(fn OnRep_ShieldOwner
  (bind _shieldowner (Variables|Default|GetShieldOwner))
  (Utilities|IsValid _shieldowner
    (:Is%20Valid
      (Transformation|AttachActorToActor self _shieldowner "None" "KeepWorld" "KeepWorld" "KeepWorld" true)
      (CallFunction|ValidateHolder)
      (bind _asfort_pawn (Utilities|Casting|CastToFortPawn _shieldowner))
      (Collision|IgnoreActorwhenMoving (Variables|/Script/Engine.Character|GetCapsuleComponent _asfort_pawn) self true))))

(fn Set_RippleFX_Param
  (params (HitWave-Loc :type Vector) (HitWave-SizeMultiplierOuterRing :type Float%20%28double-precision%29) (HitWave-SizeMultiplierInnerRing :type Float%20%28double-precision%29) (HitWave-DurationNew :type Float%20%28double-precision%29) (HitWave-Normal :type Vector))
  (bind _shieldmesh (Variables|Default|GetShieldMesh))
  (Rendering|Material|SetVectorParameterValueonMaterials _shieldmesh "HitWave-Loc" HitWave-Loc)
  (Rendering|Material|SetScalarParameterValueonMaterials _shieldmesh "GameTime" (Utilities|Time|GetGameTimeinSeconds))
  (Rendering|Material|SetScalarParameterValueonMaterials _shieldmesh "HitWave-SizeMultiplierOuterRing" HitWave-SizeMultiplierOuterRing)
  (Rendering|Material|SetScalarParameterValueonMaterials _shieldmesh "HitWave-SizeMultiplierInnerRing" HitWave-SizeMultiplierInnerRing)
  (Rendering|Material|SetScalarParameterValueonMaterials _shieldmesh "HitWave-DurationNew" HitWave-DurationNew)
  (Rendering|Material|SetVectorParameterValueonMaterials (Variables|Default|GetCoreMesh) "HitWave-Normal" HitWave-Normal))

(fn ShouldDie
  (params (Damage :type Float%20%28single-precision%29) (EventInstigator :type Controller%20Object%20Reference) (DamageCauser :type Actor%20Object%20Reference) (bIsTest :type Boolean) (EffectContext :type Gameplay%20Effect%20Context%20Handle%20Structure%20%28by%20ref%29))
  (return false))

(fn ConstructionScript)
