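# Translucent, unlit, two-sided material for the PuffyStraw ability's shield effect, with animated triangle-lattice patterns, cracked-glass detail and glowing edges.
# Timestamp-driven fade and hit-impact controls, each with a debug static switch, feed transition masks and atlas-curve flashes.
# Emissive blends cloud and crack colors with their corrupt variants, smoke and Fresnel, using HDR remapping outside mobile.
# Lattice textures drive bump-offset refraction and vertex-color-weighted world position offset.
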Material /NPCLibrary/Abilities/PuffyStraw/FX/M_FX_PuffyStraw_v2
graph {
 properties {
  bTangentSpaceNormal=false
  blendMode=BLEND_Translucent
  pixelDepthOffsetMode=PDOM_Legacy
  refractionMethod=RM_IndexOfRefraction
  shadingModel=MSM_Unlit
  translucencyLightingMode=TLM_Surface
  twoSided=true
  usedShadingModels=Unlit
 }
 nodes [
  Add(A=NamedRerouteUsage(), B=NamedRerouteUsage())
  n1 = VertexNormalWS()
  n2 = NamedRerouteUsage()
  n3 = ScalarParameter() {defaultValue=0.333, parameterName=WPO Offset}
  n4 = LinearInterpolate(A=ScalarParameter() {defaultValue=150, parameterName=Max WPO Offset}, Alpha=Clamp(None=Add(A=@n2, B=@n3) {constB=0.25})) {constB=0}
  n5 = Constant()
  n6 = TextureSample(UVs=Rotator(Coordinate=TextureCoordinate() {coordinateIndex=1, uTiling=3, vTiling=1.5}) {speed=0.025}) {samplerType=SAMPLERTYPE_Normal, texture=/Game/Weapons/WeaponSkins/Wraps/Textures/TriangleLattice/T_Wrap_TriangleLattice_N}
  n7 = Multiply(A=@n6.G, B=ScalarParameter() {defaultValue=20, parameterName=Y Intensity})
  n8 = ScalarParameter() {defaultValue=10, parameterName=Z Intensity}
  Add(A=AppendVector(A=AppendVector(A=@n5, B=@n7), B=Multiply(A=@n6.R, B=@n8)), B=Multiply(A=Multiply(A=@n1, B=@n4), B=VertexColor().R))
  n9 = Constant() {r=0.94}
  n10 = ScalarParameter() {defaultValue=1.05, parameterName=Refraction Strength}
  n11 = ScalarParameter() {defaultValue=1, parameterName=SepTrans}
  n12 = BumpOffset(
   Coordinate=Rotator(Coordinate=TextureCoordinate() {coordinateIndex=1, uTiling=3, vTiling=1.5}) {speed=0.025}
   Height=ScalarParameter() {defaultValue=0.25, parameterName=Referaction Height}
  ) {referencePlane=1}
  n13 = NamedRerouteUsage()
  n14 = NamedRerouteUsage()
  LinearInterpolate(
   B=LinearInterpolate(
    B=LinearInterpolate(A=@n9, B=@n10, Alpha=@n11)
    Alpha=TextureSample(UVs=@n12) {samplerType=SAMPLERTYPE_Masks, texture=/Game/Weapons/WeaponSkins/Wraps/Textures/TriangleLattice/T_Wrap_TriangleLattice_M}.G
   ) {constA=1, constAlpha=0}
   Alpha=Saturate(None=Multiply(A=@n13, B=@n14))
  ) {constA=1, desc=Transition}
  n15 = Transform(None=@n6.RGB)
  n16 = Multiply(
   A=MaterialFunctionCall(Exponent=ScalarParameter() {defaultValue=0.25, parameterName=Fresnel Exponent}, Vertex / Pixel Normal=@n15) {materialFunction=&MF_Fresnel_FX}.Fresnel Inverted Double Sided
   B=ScalarParameter() {defaultValue=0.9, parameterName=Fresnel Opacity}
  ) {constB=0.9}
  n17 = ScalarParameter() {defaultValue=0.25, parameterName=Emissive Boost}
  n18 = Add(A=Add(A=NamedRerouteUsage(), B=Add(A=@n16, B=@n17)), B=NamedRerouteUsage())
  n19 = MaterialFunctionCall(Distance=ScalarParameter() {defaultValue=100, parameterName=Near Fade}) {materialFunction=&Near_Camera_Fade}
  Multiply(A=Multiply(A=Multiply(A=@n18, B=@n19.Result), B=NamedRerouteUsage()), B=NamedRerouteUsage())
  n20 = ScalarParameter() {defaultValue=1, parameterName=Transition}
  n21 = ScalarParameter() {parameterName=FadeInTime}
  n22 = ScalarParameter() {defaultValue=0.666, parameterName=Fade In Length}
  n23 = ScalarParameter() {parameterName=FadeOutTime}
  n24 = ScalarParameter() {defaultValue=0.333, parameterName=Fade Out Length}
  n25 = StaticBool()
  NamedRerouteDeclaration(None=StaticSwitchParameter(
    True=Saturate(None=@n20)
    False=MaterialFunctionCall(TimestampA=@n21, TransitionLength_In=@n22, TimestampB=@n23, TransitionLength_Out=@n24, Transition In Only?=@n25) {materialFunction=&MF_Timestamp_InOut_01}.Linear
   ) {parameterName=DEBUG CONTROL}) {name=Control, nodeColor=<rgba 0.875 0 1 1>}
  n26 = TextureCoordinate()
  n27 = Multiply(A=ScalarParameter() {defaultValue=1, parameterName=Grid U Repeat}, B=ScalarParameter() {defaultValue=1, parameterName=Radius})
  n28 = AppendVector(A=@n27, B=ScalarParameter() {defaultValue=1, parameterName=Grid V Repeat})
  n29 = Subtract(A=TextureSample(UVs=Multiply(A=@n26, B=@n28)) {samplerType=SAMPLERTYPE_Masks, texture=&T_Wrap_TriangleLattice_M}.G)
  n30 = AppendVector(A=Constant() {r=1}, B=Multiply(A=@n27) {constB=2})
  n31 = TextureCoordinate() {vTiling=2}
  n32 = Sine(None=Time())
  n33 = Divide(A=Ceil(None=Multiply(A=@n32) {constB=4}))
  n34 = AppendVector(A=LinearInterpolate(Alpha=NamedRerouteUsage()) {constA=-1, constB=0.5}, B=Constant())
  n35 = Add(A=Multiply(A=@n30, B=Panner(Coordinate=@n31, Time=@n33)), B=@n34)
  n36 = Multiply(
   A=TextureSample(UVs=MaterialFunctionCall(XY=@n35) {materialFunction=&Swizzle}.YX) {texture=&/T_PuffyStrawShield}.B
   B=ScalarParameter() {defaultValue=1.5, parameterName=Transition Edge Brightness}
  ) {constB=1.5}
  NamedRerouteDeclaration(None=Clamp(None=Add(A=@n29, B=@n36))) {name=Transition Opacity Output A, nodeColor=<rgba 0 1 0.3438 1>}
  n37 = BumpOffset(Coordinate=TextureCoordinate() {coordinateIndex=1, uTiling=2}, Height=ScalarParameter() {defaultValue=-0.5, parameterName=Height}) {referencePlane=1}
  n38 = Add(
   A=TextureSample(UVs=@n37) {texture=/Game/Effects/Fort_Effects/Textures/Tile/T_Glass_Cracks_M}.G
   B=Multiply(A=TextureSample(UVs=@n37) {texture=&T_Glass_Cracks_M}.G) {constB=0.1}
  )
  n39 = OneMinus(None=VertexColor().G)
  NamedRerouteDeclaration(None=Multiply(A=Multiply(A=@n38, B=@n39), B=NamedRerouteUsage())) {name=Shattered, nodeColor=<rgba 0 0.4375 1 1>}
  n40 = TextureCoordinate() {coordinateIndex=1}
  n41 = TextureSample(UVs=@n40) {texture=&T_PuffyStrawShield}
  n42 = Add(
   A=Add(A=TextureSample(UVs=@n40) {texture=&T_PuffyStrawShield}.R, B=Multiply(A=@n41.R) {constB=0.1})
   B=Multiply(A=VertexColor().G) {constB=0.1}
  )
  n43 = ScalarParameter() {defaultValue=0.25, parameterName=Join Intensity}
  NamedRerouteDeclaration(None=Multiply(A=Multiply(A=@n42, B=@n43), B=NamedRerouteUsage())) {name=Joint, nodeColor=<rgba 1 0 0.3281 1>}
  n44 = VectorParameter() {parameterName=CorruptCloudColor, defaultValue=<rgba 1 0.2747 0 1>}
  n45 = TextureCoordinate() {coordinateIndex=1, uTiling=2.5}
  n46 = TextureSample(UVs=BumpOffset(Coordinate=Rotator(Coordinate=@n45) {speed=0.025}, Height=ScalarParameter() {defaultValue=-0.5, parameterName=Height}) {referencePlane=1}) {samplerType=SAMPLERTYPE_Masks, texture=&T_Wrap_TriangleLattice_M}
  n47 = Clamp(None=Multiply(A=Power(Base=@n46.G) {constExponent=4}) {constB=5})
  n48 = LinearInterpolate(A=VectorParameter() {parameterName=CloudColor, defaultValue=<rgba 1 0.0252 0 1>}.RGB, B=@n44.RGB, Alpha=@n47)
  NamedRerouteDeclaration(None=@n48) {name=Cloud Color, nodeColor=<rgba 1 0.9609 0 1>}
  n49 = VectorParameter() {parameterName=CorruptCrackColor, defaultValue=<rgba 0.0908 0.0222 0.147 1>}
  n50 = LinearInterpolate(A=VectorParameter() {parameterName=CrackColor, defaultValue=<rgba 1 0.013 0.0452 1>}.RGB, B=@n49.RGB, Alpha=@n47) {constAlpha=1}
  NamedRerouteDeclaration(None=@n50) {name=Crack Color, nodeColor=<rgba 0.9922 0 1 1>}
  n51 = TextureCoordinate() {vTiling=0.75}
  n52 = Add(A=Multiply(A=@n30, B=Panner(Coordinate=@n51) {speedY=0.02}), B=@n34)
  NamedRerouteDeclaration(None=TextureSample(UVs=MaterialFunctionCall(XY=@n52) {materialFunction=&Swizzle}.YX) {texture=&T_PuffyStrawShield}.G) {name=Transition Opacity Output B, nodeColor=<rgba 0.9453 0 1 1>}
  n53 = ScalarParameter() {defaultValue=1, parameterName=Hit Impact Transition}
  n54 = ScalarParameter() {parameterName=HitFadeInTime}
  n55 = ScalarParameter() {defaultValue=0.1, parameterName=Hit Fade Length}
  n56 = ScalarParameter() {parameterName=HitFadeOutTime}
  n57 = StaticBool()
  NamedRerouteDeclaration(None=StaticSwitchParameter(
    True=Saturate(None=@n53)
    False=MaterialFunctionCall(TimestampA=@n54, TransitionLength_In=@n55, TimestampB=@n56, TransitionLength_Out=@n55, Transition In Only?=@n57) {materialFunction=&MF_Timestamp_InOut_01}.Linear
   ) {parameterName=Hit Immpact DEBUG CONTROL}) {name=Hit Impact Flash, nodeColor=<rgba 0.875 0 1 1>}
  n58 = TextureSample(UVs=TextureCoordinate() {coordinateIndex=1}) {texture=&T_PuffyStrawShield}
  n59 = Power(Base=@n58.A, Exp=ScalarParameter() {defaultValue=5, parameterName=Edge Contrast}) {constExponent=1}
  n60 = TextureSample(UVs=Panner(Coordinate=TextureCoordinate() {coordinateIndex=1, uTiling=2, vTiling=2}) {speedY=0.0625}) {texture=/Game/Weapons/WeaponSkins/Wraps/Textures/TNTina/T_TNTina_TriangleNoise}
  n61 = Add(A=@n60.R, B=@n60.R)
  n62 = ScalarParameter() {defaultValue=5, parameterName=Triangle Contrast}
  n63 = ScalarParameter() {defaultValue=20, parameterName=Tri Contrast}
  n64 = Multiply(A=Subtract(A=@n58.A, B=Power(Base=@n58.A, Exp=@n63) {constExponent=20}), B=ScalarParameter() {defaultValue=100, parameterName=IntensityUp})
  n65 = Add(A=@n59, B=LinearInterpolate(B=Power(Base=@n61, Exp=@n62), Alpha=Saturate(None=@n64)))
  NamedRerouteDeclaration(None=Saturate(None=Multiply(A=@n58.A, B=@n65))) {name=Edge Opacity, nodeColor=<rgba 0 1 0.7891 1>}
  n66 = TextureCoordinate() {coordinateIndex=1}
  n67 = Subtract(A=OneMinus(None=TextureSample(UVs=@n66) {texture=&T_PuffyStrawShield}.R), B=NamedRerouteUsage())
  n68 = VectorParameter() {parameterName=LargeShardTint, defaultValue=<rgba 0.0343 0 0.1878 1>}
  n69 = TextureCoordinate() {coordinateIndex=1, uTiling=4, vTiling=1.2}
  n70 = Multiply(A=Time()) {constB=1.2}
  n71 = Constant()
  n72 = ShadingPathSwitch(
   Default=TextureSample(UVs=Panner(Coordinate=@n69, Time=@n70) {speedX=0.0312, speedY=0.0625}) {texture=&T_Inky_Smoke_Tile_Chromatic}.RGB
   Forward=@n71
   Mobile=@n71
  )
  n73 = LinearInterpolate(Alpha=NamedRerouteUsage()) {constA=0.22}
  n74 = Multiply(A=Multiply(A=Add(A=@n72, B=@n73)) {constB=0.25}, B=Multiply(A=Add(A=@n50) {constA=0.2, constB=0.1}) {constB=2})
  n75 = Multiply(A=Multiply(A=@n48) {constB=0.5}, B=NamedRerouteUsage())
  n76 = Add(
   A=@n47
   B=MaterialFunctionCall(
    Brightness=ScalarParameter() {defaultValue=1, parameterName=FresBirghtness}
    Exponent=ScalarParameter() {defaultValue=2, parameterName=Vertex COntrast}
   ) {materialFunction=&MF_Fresnel_FX}.Fresnel Double Sided
  )
  n77 = Multiply(A=LinearInterpolate(A=@n49.RGB, B=@n44.RGB, Alpha=Saturate(None=@n76))) {constB=8}
  n78 = NamedRerouteUsage()
  n79 = Multiply(A=Add(A=Add(A=@n74, B=@n75), B=Multiply(A=@n77, B=@n78)), B=ScalarParameter() {defaultValue=0.5, parameterName=Emissive})
  n80 = MaterialFunctionCall(LDR Input=Saturate(None=@n79), HDR Tint=@n79, HDR Intensity Multiplier=ScalarParameter() {defaultValue=1, parameterName=Base HDR Intensity}) {materialFunction=&DeriveHDRfromLDR}
  NamedRerouteDeclaration(None=Add(A=Multiply(A=@n67, B=@n68.RGB), B=ShadingPathSwitch(Default=@n80.RemappedRange, Mobile=@n79))) {name=Emissive, nodeColor=<rgba 1 0.426 0 1>}
  NamedRerouteDeclaration(None=@n59) {name=Edge Alpha Mask, nodeColor=<rgba 0 1 1 1>}
  n81 = Constant3Vector()
  n82 = NamedRerouteUsage()
  n83 = NamedRerouteUsage()
  n84 = Multiply(A=NamedRerouteUsage(), B=NamedRerouteUsage())
  n85 = CurveAtlasRowParameter(CurveTime=NamedRerouteUsage()) {
   atlas=&/PuffyStrawAtlas
   curve=&/PuffyStraw_FlashCurve
   parameterName=PuffyStrawAtlas
  }
  n86 = LinearInterpolate(
   A=NamedRerouteUsage()
   B=Multiply(A=NamedRerouteUsage(), B=ScalarParameter() {defaultValue=10, parameterName=Flash Intensity})
   Alpha=VertexColor().G
  )
  n87 = NamedRerouteUsage()
  n88 = Add(
   A=Add(
    A=MaterialFunctionCall(A=@n81, B=@n82, C=@n83, Alpha=@n84) {materialFunction=&3ColorBlend}.Result
    B=Multiply(A=@n85.R, B=@n86)
   )
   B=Multiply(A=Multiply(A=@n86, B=@n87), B=ScalarParameter() {defaultValue=0.5, parameterName=Hit Impact Intensity Bias})
  )
  n89 = Saturate(None=@n88)
  n90 = ScalarParameter() {defaultValue=1, parameterName=HDR Intensity}
  NamedRerouteDeclaration(None=ShadingPathSwitch(
    Default=MaterialFunctionCall(LDR Input=@n89, HDR Tint=@n88, HDR Intensity Multiplier=@n90) {materialFunction=&DeriveHDRfromLDR}.RemappedRange
    Mobile=@n88
   )) {name=Flash Color, nodeColor=<rgba 1 0.2812 0 1>}
  n91 = NamedRerouteUsage()
  n92 = LinearInterpolate(Alpha=@n91)
  n93 = Multiply(A=@n92, B=Add(A=@n92, B=@n91))
  n94 = ScalarParameter() {defaultValue=1, parameterName=Opacity}
  NamedRerouteDeclaration(None=Saturate(None=Multiply(A=@n93, B=@n94))) {name=Opacity, nodeColor=<rgba 0 0.7422 1 1>}
  Reroute(None=@n15)
 ]
}
